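# Looping SoundCue for the Season 26 Pointing Ghost quest chest, blending treasure chest ambience with ghost sounds.
# A randomly picked chest ambient loop fades in through an Enveloper and mixes with two continuous ghost loops.
# A Looping branch adds one of seven random ghost one-shots after random delays of up to two seconds.
# It plays on the quests sound class with overridden short-range logarithmic attenuation.
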SoundCue /BattlePassS26_Quests/Sound/Quest_S26_PointingGhost/S26_PointingGhost_Chest_LP
root=Mixer_0
nodes [
 {id=Delay_0, children [Random_0], properties {delayMax=2}}
 {
  id=Enveloper_0
  children [Random_2]
  properties {pitchCurve {editorCurveData {keys [{time=0}]}}, volumeCurve {editorCurveData {keys [<at 0 0>,<at 1 1>]}}}
 }
 {id=Looping_0, children [Delay_0]}
 {id=Mixer_0, children [Mixer_1,Looping_0], properties {inputVolume [1,0.75]}}
 {id=Mixer_1, children [Enveloper_0,WavePlayer_0,WavePlayer_8], properties {inputVolume [0.65,0.75,1]}}
 {
  id=Random_0
  children [WavePlayer_1,WavePlayer_2,WavePlayer_3,WavePlayer_4,WavePlayer_5,WavePlayer_6,WavePlayer_7]
  properties {weights [1,1,1,1,1,1,1]}
 }
 {id=Random_2, children [WavePlayer_12, WavePlayer_13], properties {bShouldExcludeFromBranchCulling=true, weights [1,1]}}
 {
  id=WavePlayer_0
  properties {bLooping=true, soundWaveAssetPtr=&/S26_PointingGhost_Q03_Ghost_LP}
 }
 {id=WavePlayer_1, properties {soundWaveAssetPtr=&/S26_PointingGhost_Q03_Ghost_Lyr_01}}
 {
  id=WavePlayer_12
  properties {bLooping=true, soundWaveAssetPtr=/Game/Sounds/Foley_Loot/Containers/Treasure_Chest/Treasure_Chest_Ambient_Loop_01}
 }
 {
  id=WavePlayer_13
  properties {bLooping=true, soundWaveAssetPtr=/Game/Sounds/Foley_Loot/Containers/Treasure_Chest/Treasure_Chest_Ambient_Loop_02}
 }
 {id=WavePlayer_2, properties {soundWaveAssetPtr=&/S26_PointingGhost_Q03_Ghost_Lyr_02}}
 {id=WavePlayer_3, properties {soundWaveAssetPtr=&/S26_PointingGhost_Q03_Ghost_Lyr_03}}
 {id=WavePlayer_4, properties {soundWaveAssetPtr=&/S26_PointingGhost_Q03_Ghost_Lyr_04}}
 {id=WavePlayer_5, properties {soundWaveAssetPtr=&/S26_PointingGhost_Q03_Ghost_Lyr_05}}
 {id=WavePlayer_6, properties {soundWaveAssetPtr=&/S26_PointingGhost_Q03_Ghost_Lyr_06}}
 {id=WavePlayer_7, properties {soundWaveAssetPtr=&/S26_PointingGhost_Q03_Ghost_Lyr_07}}
 {
  id=WavePlayer_8
  properties {bLooping=true, soundWaveAssetPtr=&/S26_PointingGhost_Q03_Ghost_LP_Lyr}
 }
]
properties {
 attenuationSettings=/Game/Sounds/Attenuation/Fort_Item_Ambient_Loop_Mono_Default_Att
 bOverrideAttenuation=true
 maxDistance=850
 soundClassObject=&Fort_Quests
 volumeMultiplier=0.7
 attenuationOverrides {distanceAlgorithm=Logarithmic, falloffDistance=800, stereoSpread=100, attenuationShapeExtents {x=50}}
}
