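# Looping idle sound cue for the Season 26 pointing ghost quest, routed to the quests sound class.
# A mixer blends two constantly looping ghost beds, slightly lowered, with an intermittent layer at full level.
# That layer loops through a short random delay into a random pick from seven equally weighted one-shot variations.
# It references another quest's attenuation asset but overrides it with logarithmic distance falloff.
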SoundCue /BattlePassS26_Quests/Sound/Quest_S26_PointingGhost/S26_PointingGhost_Idle_LP
root=Mixer_0
nodes [
 {id=Delay_0, children [Random_0], properties {delayMax=2}}
 {id=Looping_0, children [Delay_0]}
 {id=Mixer_0, children [Mixer_1,Looping_0], properties {inputVolume [0.75,1]}}
 {id=Mixer_1, children [WavePlayer_0,WavePlayer_8], properties {inputVolume [1,1]}}
 {
  id=Random_0
  children [WavePlayer_1,WavePlayer_2,WavePlayer_3,WavePlayer_4,WavePlayer_5,WavePlayer_6,WavePlayer_7]
  properties {weights [1,1,1,1,1,1,1]}
 }
 {
  id=WavePlayer_0
  properties {bLooping=true, soundWaveAssetPtr=&/S26_PointingGhost_Q03_Ghost_LP}
 }
 {id=WavePlayer_1, properties {soundWaveAssetPtr=&/S26_PointingGhost_Q03_Ghost_Lyr_01}}
 {id=WavePlayer_2, properties {soundWaveAssetPtr=&/S26_PointingGhost_Q03_Ghost_Lyr_02}}
 {id=WavePlayer_3, properties {soundWaveAssetPtr=&/S26_PointingGhost_Q03_Ghost_Lyr_03}}
 {id=WavePlayer_4, properties {soundWaveAssetPtr=&/S26_PointingGhost_Q03_Ghost_Lyr_04}}
 {id=WavePlayer_5, properties {soundWaveAssetPtr=&/S26_PointingGhost_Q03_Ghost_Lyr_05}}
 {id=WavePlayer_6, properties {soundWaveAssetPtr=&/S26_PointingGhost_Q03_Ghost_Lyr_06}}
 {id=WavePlayer_7, properties {soundWaveAssetPtr=&/S26_PointingGhost_Q03_Ghost_Lyr_07}}
 {
  id=WavePlayer_8
  properties {bLooping=true, soundWaveAssetPtr=&/S26_PointingGhost_Q03_Ghost_LP_Lyr}
 }
]
properties {
 attenuationSettings=/BattlePassS26_Quests/Sound/Quest_S26_BP_CharacterToken_Q01/IP_Quest_S26_BP_CharacterToken_Q01_Atten
 bOverrideAttenuation=true
 maxDistance=1350
 soundClassObject=&Fort_Quests
 volumeMultiplier=0.8
 attenuationOverrides {distanceAlgorithm=Logarithmic, falloffDistance=1250, stereoSpread=400, attenuationShapeExtents {x=100}}
}
